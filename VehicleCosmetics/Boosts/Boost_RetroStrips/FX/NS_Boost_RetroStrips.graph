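# Niagara system for the RetroStrips vehicle boost cosmetic, placing spinning pyramid meshes, sprites and ribbon trails at boost sockets.
# Pyramid burst-spawns two groups, partitioned across ports, with rotation varying by spawn group and partition.
# Three near-identical Retrotrail emitters spawn per-port ribbons with painted-variant colors, backward velocity and source-movement inheritance fading over lifetime.
# Each trail also computes four sine-wobbled "PosNoodle" positions, mostly widening with particle age.
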NiagaraSystem /VehicleCosmetics/Boosts/Boost_RetroStrips/FX/NS_Boost_RetroStrips
dependencies {
 usedDataInterfaces [&NiagaraDataInterfaceArrayFloat, &NiagaraDataInterfaceCurve, &NiagaraDataInterfaceSkeletalMesh]
 usedDynamicInputs [&RandomRangeFloat]
 usedModules [
  &self:SystemSpawnScript.SystemScriptSource.SystemScriptGraph.NiagaraNodeAssignment_0.SetVariables_#1
  &SystemState
 ]
}
emitters {
 Pyramid {
  emitterUpdateScript [
   &EmitterState(
    Inactive Response=::NewEnumerator1
    Life Cycle Mode=::NewEnumerator1
    Loop Behavior=::NewEnumerator1
    Loop Duration Mode=::NewEnumerator1
    UseLoopDelay=false
   )
   &SpawnBurst_Instantaneous(Spawn Count=4, Use Spawn Probability=false, UseLoopCountLimit=false)
   &SpawnBurst_Instantaneous(Spawn Count=4, Spawn Group=1, Use Spawn Probability=false, UseLoopCountLimit=false) {module=SpawnBurst_Instantaneous001}
   &VC_FilteredSocketCount()
   &SpawnRate(SpawnRate=80, Use Spawn Probability=false) {enabled=false}
  ]
  particleSpawnScript [
   /SetVariables(Particles.VisTag:NiagaraInt32=&SpawnGroup()) {isSetParameters=true}
   &VC_PaintedVariantsOverride(
    Black=<rgba 0.1302 0.1302 0.1302 1>
    BurntSienna=<rgba 0.2552 0.0865 0.0412 1>
    Cobalt=<rgba 0 0.0555 3 1>
    Crimson=<rgba 1 0 0.0271 1>
    Pink=<rgba 0.9844 0.1179 0.6693 1>
   ) {enabled=false}
   &InitializeParticle(
    Lifetime=$Emitter.CurrentLoopDuration
    Mesh Scale Mode=::NewEnumerator3
    Mesh Uniform Scale=0.26
    Position Offset=<300 0 0>
    Sprite Rotation Mode=::NewEnumerator2
    UsePositionOffset=false
    Write Material Random=true
   )
   &PartitionParticles(Number Of Partitions=$Emitter.PortCount, Particles Per Partition=2, Partition Mode:ENiagara_PartitionMode=::NewEnumerator1)
   &SkeletalMeshLocation(
    Bone / Socket Index=$Particles.PartitionParticles.Partition
    Bone / Socket Sampling Mode=::NewEnumerator5
    MaskSpawnGroup=false
    Mesh Sampling Type=::NewEnumerator6
    Sampled Position Offset=<-50 0 0>
    Skeletal Mesh=$System.SkeletalMesh
   )
   &AddVelocity()
  ]
  particleUpdateScript [
   &ParticleState()
   &UpdateMeshOrientation(Rotation Vector=&MaskVectorBySpawnGroup(Masked Vector=<0 0 1>, Passthrough Vector=<0 0 -1>)) {enabled=false}
   &UpdateMeshOrientation(
    Rotation Rate=/VehicleCosmetics/FX/Niagara/Modules/MergeEmitters/VC_FloatBySpawnGroup(
     SpawnGroup or ExecIndex:NiagaraInt32=0
     CustomInt:NiagaraInt32=0
     MaximumSpawnGroups:NiagaraInt32=2
     Float SG 0:NiagaraFloat=&SelectFloatFromArray(Direct Array Index=$Particles.PartitionParticles.Partition, Float Selection Array=DataInterface() {propertyValues {FloatData [0.4,-0.4]}})
     Float SG 1:NiagaraFloat=&SelectFloatFromArray(Direct Array Index=$Particles.PartitionParticles.Partition, Float Selection Array=DataInterface() {propertyValues {FloatData [0.4,-0.4]}})
     Float SG 2:NiagaraFloat=0
     Float SG 3:NiagaraFloat=0
     Float SG 4:NiagaraFloat=0
     Float SG 5:NiagaraFloat=0
    )
    Rotation Vector=/VehicleCosmetics/FX/Niagara/Modules/MergeEmitters/VC_VectorBySpawnGroup(
     SpawnGroup or ExecIndex:NiagaraInt32=0
     CustomInt:NiagaraInt32=0
     MaximumSpawnGroups:NiagaraInt32=2
     Vector Group 0:Vector3f=<0 0 -1>
     Vector Group 1:Vector3f=<0 0 1>
     Vector Group 2:Vector3f=<0 0 0>
     Vector Group 3:Vector3f=<0 0 0>
     Vector Group 4:Vector3f=<0 0 0>
     Vector Group 5:Vector3f=<0 0 0>
    )
   ) {module=UpdateMeshOrientation001}
   &ScaleVelocity(Velocity Scale=&VectorFromFloat(Value=&VC_FloatBySpawnGroup(
      SpawnGroup or ExecIndex:NiagaraInt32=0
      CustomInt:NiagaraInt32=0
      MaximumSpawnGroups:NiagaraInt32=2
      Float SG 0:NiagaraFloat=&Sine(
       Normalized Angle=$Emitter.Age
       Period=2
       Scale=&SelectFloatFromArray(Direct Array Index=$Particles.PartitionParticles.Partition, Float Selection Array=DataInterface() {propertyValues {FloatData [0.2,0.3]}})
       Bias=0
      )
      Float SG 1:NiagaraFloat=&Sine(
       Normalized Angle=$Emitter.Age
       Period=2
       Scale=&SelectFloatFromArray(Direct Array Index=$Particles.PartitionParticles.Partition, Float Selection Array=DataInterface() {propertyValues {FloatData [0.3,0.2]}})
       Bias=0
      )
      Float SG 2:NiagaraFloat=0
      Float SG 3:NiagaraFloat=0
      Float SG 4:NiagaraFloat=0
      Float SG 5:NiagaraFloat=0
     )))
   &SolveForcesAndVelocity(Clamp Velocity=false, Limit Acceleration=false)
   &ScaleColor(Scale RGB=&VectorFromFloat(Value=1), ScaleA=true, ScaleRGB=true)
  ]
  rendererClasses [&NiagaraMeshRendererProperties, &NiagaraSpriteRendererProperties]
  renderers [
   {rendererClass=&NiagaraMeshRendererProperties, rendererIndex=0}
   {rendererClass=&NiagaraMeshRendererProperties, rendererIndex=1}
   {rendererClass=&NiagaraSpriteRendererProperties, rendererIndex=2}
  ]
 }
 Retrotrail1001 {
  emitterUpdateScript [
   &EmitterState()
   &SpawnPerUnit(Module.Use Max Movement Threshold=true, Module.Use Movement Tolerance=false, Use Spawn Probability=false) {enabled=false}
   &VC_FilteredSocketCount()
   &VC_RateQualityHandler()
   &VC_SpawnRibbon_v1(Number of Ribbons=$Emitter.PortCount, SpawnRate=60)
  ]
  particleSpawnScript [
   &VC_PaintedVariantsOverride(
    Black=<rgba 0.0156 0.0156 0.0156 1>
    BurntSienna=<rgba 0.1562 0.0504 0.022 0.8>
    Cobalt=<rgba 0.0522 0.138 0.9115 0.8348>
    Crimson=<rgba 0.4271 0.0356 0.0462 0.5217>
    DefaultColor=<rgba 0.1284 0.3065 1 1>
    ForestGreen=<rgba 0 0.5 0.0253 0.4>
    Grey=<rgba 0.2083 0.2083 0.2083 1>
    Lime=<rgba 0.4075 0.6094 0.0508 0.8522>
    Orange=<rgba 0.6724 0.2346 0 0.8>
    Pink=<rgba 0.7 0.1896 0.5144 0.8>
    Purple=<rgba 0.1913 0 0.6795 1>
    R1 Black:LinearColor=<rgba 0.2708 0.2708 0.2708 1>
    R1 BurntSienna:LinearColor=<rgba 0.8229 0.2696 0.0129 1>
    R1 Cobalt:LinearColor=<rgba 0 0.185 10 1>
    R1 Crimson:LinearColor=<rgba 1 0 0.0271 1>
    R1 DefaultColor=<rgba 4 0.28 0.28 1>
    R1 Grey:LinearColor=<rgba 0.7478 0.7478 0.7478 1>
    R1 Lime:LinearColor=<rgba 1.5 3 0 1>
    R1 Orange:LinearColor=<rgba 5 1.3346 0 1>
    R1 Pink:LinearColor=<rgba 3 0.5 2.0909 1>
    R1 Purple:LinearColor=<rgba 0.4 0 2 1>
    R1 Saffron:LinearColor=<rgba 3 2.3 0 1>
    R1 SkyBlue:LinearColor=<rgba 0 1.9962 3 1>
    R1 TitaniumWhite:LinearColor=<rgba 1 1 1 1>
    R2 Black:LinearColor=<rgba 0.0365 0.0365 0.0365 1>
    R2 BurntSienna:LinearColor=<rgba 0.2083 0.0709 0.0335 1>
    R2 DefaultColor:LinearColor=<rgba 0.1284 0.3065 1 1>
    Saffron=<rgba 0.8698 0.6774 0.0453 0.6957>
    SkyBlue=<rgba 0 0.6654 1 1>
    TitaniumWhite=<rgba 0.8854 0.8854 0.8854 1>
    Use R1 Painted Overrides=true
    Use R2 Painted Overrides=true
   )
   &VC_InitializeRibbon_v1(
    Color=$Particles.Color
    Custom Offset=true
    Enable Skeletal Position=true
    Lifetime=0.8
    Local Vector Offset=<-100 0 0>
    Offset Amount=1
    Ribbon Width=60
   )
   &AddVelocity(Velocity=<-1200 0 0>, Velocity Speed Scale=1.4) {module=AddVelocity001}
  ]
  particleUpdateScript [
   &ParticleState()
   &InheritSourceMovement(
    Applied Position Scale=&Multiply_VectorByFloat(
     Vector=<1 1 1>
     Float=&FloatFromCurve(FloatCurve=DataInterface() {propertyValues {Curve {
         defaultValue=1
         keys [{interpMode=RCIM_Cubic, tangentMode=RCTM_User, time=0, value=1}, {interpMode=RCIM_Cubic, tangentMode=RCTM_User, time=1, value=0}]
        }}})
    )
    ApplyForce=false
    ApplyPosition=true
    ApplyPositionFalloff=false
   )
   &SolveForcesAndVelocity(Clamp Velocity=false, Limit Acceleration=false)
   &VC_UpdateRibbon_v1(Enable Skeletal Position=true)
   /SetVariables(
    Particles.PosNoodle1:NiagaraPosition=&AddVectorToPosition(
     Position=$Particles.Position
     Vector=&TransformVector(
      Vector=&MakeVector(
       X=&Sine(
        Normalized Angle=&Add_Float(
         A=&Multiply_Float(A=$Particles.NormalizedAge, B=-1.3)
         B=&Multiply_Float(A=$Emitter.LoopedAge, B=1.2)
        )
        Period=1
        Scale=40
        Bias=0
       )
       Y=&Sine(
        Normalized Angle=&Add_Float(
         A=&Multiply_Float(A=$Particles.NormalizedAge, B=-1.3)
         B=&Multiply_Float(A=$Emitter.LoopedAge, B=1)
        )
        Period=1.5
        Scale=5
        Bias=0
       )
       Z=&Sine(
        Normalized Angle=&Add_Float(
         A=&Multiply_Float(A=$Particles.NormalizedAge, B=-1.3)
         B=&Multiply_Float(A=$Emitter.LoopedAge, B=0.8)
        )
        Period=1.5
        Scale=5
        Bias=0
       )
      )
      Destination Space=::Simulation
      Apply Scale=true
     )
    )
    Particles.PosNoodle2:NiagaraPosition=&AddVectorToPosition(
     Position=$Particles.Position
     Vector=&TransformVector(
      Vector=&MakeVector(
       X=0
       Y=&Sine(
        Normalized Angle=&Add_Float(
         A=&Multiply_Float(A=$Particles.NormalizedAge, B=-1.3)
         B=&Multiply_Float(A=$Emitter.LoopedAge, B=1.8)
        )
        Period=0.4
        Scale=&FloatFromCurve(FloatCurve=DataInterface() {propertyValues {Curve {keys [
             {arriveTangent=41.4286, interpMode=RCIM_Cubic, time=0, value=0.3}
             {arriveTangentWeight=3.3375, interpMode=RCIM_Cubic, leaveTangentWeight=3.3375, tangentMode=RCTM_User, time=0.3, value=20}
             {arriveTangent=28.1747, interpMode=RCIM_Cubic, leaveTangent=28.1747, tangentMode=RCTM_User, time=1, value=50}
            ]}}})
        Bias=0
       )
       Z=&Sine(
        Normalized Angle=&Add_Float(
         A=&Multiply_Float(A=$Particles.NormalizedAge, B=-1.3)
         B=&Multiply_Float(A=$Emitter.LoopedAge, B=2)
        )
        Period=0.8
        Scale=&FloatFromCurve(FloatCurve=DataInterface() {propertyValues {Curve {keys [
             <at 0.3 0.3 RCIM_Cubic>
             {
              arriveTangent=10.3659
              arriveTangentWeight=3.3375
              interpMode=RCIM_Cubic
              leaveTangent=10.3659
              leaveTangentWeight=3.3375
              tangentMode=RCTM_User
              time=0.5
              value=20
             }
             {arriveTangent=75.1908, interpMode=RCIM_Cubic, leaveTangent=75.1908, tangentMode=RCTM_User, time=1, value=60}
            ]}}})
        Bias=0
       )
      )
      Destination Space=::Simulation
      Apply Scale=true
     )
    )
    Particles.PosNoodle3:NiagaraPosition=&AddVectorToPosition(
     Position=$Particles.Position
     Vector=&TransformVector(
      Vector=&MakeVector(
       X=0
       Y=&Sine(
        Normalized Angle=&Add_Float(
         A=&Multiply_Float(A=$Particles.NormalizedAge, B=-1.3)
         B=&Multiply_Float(A=$Emitter.LoopedAge, B=2.1)
        )
        Period=0.5
        Scale=&FloatFromCurve(FloatCurve=DataInterface() {propertyValues {Curve {keys [
             {arriveTangent=39.5768, interpMode=RCIM_Cubic, leaveTangent=39.5768, tangentMode=RCTM_User, time=0, value=0.3}
             {
              arriveTangent=32.917
              arriveTangentWeight=3.3375
              interpMode=RCIM_Cubic
              leaveTangent=32.917
              leaveTangentWeight=3.3375
              time=0.595
              value=8.803
             }
             {interpMode=RCIM_Cubic, leaveTangent=75.1908, time=1, value=33.217}
            ]}}})
        Bias=0
       )
       Z=&Sine(
        Normalized Angle=&Add_Float(
         A=&Multiply_Float(A=$Particles.NormalizedAge, B=-1.3)
         B=&Multiply_Float(A=$Emitter.LoopedAge, B=1.7)
        )
        Period=0.5
        Scale=&FloatFromCurve(FloatCurve=DataInterface() {propertyValues {Curve {keys [
             <at 0 0 RCIM_Cubic>
             {arriveTangentWeight=3.3375, interpMode=RCIM_Cubic, leaveTangentWeight=3.3375, tangentMode=RCTM_User, time=0.3, value=1}
             {arriveTangent=75.1908, interpMode=RCIM_Cubic, leaveTangent=75.1908, tangentMode=RCTM_User, time=1, value=45}
            ]}}})
        Bias=0
       )
      )
      Destination Space=::Simulation
      Apply Scale=true
     )
    )
    Particles.PosNoodle4:NiagaraPosition=&AddVectorToPosition(
     Position=$Particles.Position
     Vector=&TransformVector(
      Vector=&MakeVector(
       X=0
       Y=&Sine(
        Normalized Angle=&Add_Float(
         A=&Multiply_Float(A=$Particles.NormalizedAge, B=-1.3)
         B=&Multiply_Float(A=$Emitter.LoopedAge, B=1.5)
        )
        Period=0.4
        Scale=&FloatFromCurve(FloatCurve=DataInterface() {propertyValues {Curve {keys [
             {arriveTangent=41.4286, interpMode=RCIM_Cubic, time=0, value=0.3}
             {arriveTangentWeight=3.3375, interpMode=RCIM_Cubic, leaveTangentWeight=3.3375, tangentMode=RCTM_User, time=0.2, value=11.4906}
             {arriveTangent=31.4052, interpMode=RCIM_Cubic, leaveTangent=31.4052, tangentMode=RCTM_User, time=1, value=26}
            ]}}})
        Bias=0
       )
       Z=&Sine(
        Normalized Angle=&Add_Float(
         A=&Multiply_Float(A=$Particles.NormalizedAge, B=-1.3)
         B=&Multiply_Float(A=$Emitter.LoopedAge, B=1)
        )
        Period=1
        Scale=&FloatFromCurve(FloatCurve=DataInterface() {propertyValues {Curve {keys [
             <at 0 1 RCIM_Cubic>
             {arriveTangentWeight=6.6125, interpMode=RCIM_Cubic, leaveTangentWeight=6.725, tangentMode=RCTM_User, time=0.6, value=20}
             {interpMode=RCIM_Cubic, leaveTangent=75.1908, time=1, value=35}
            ]}}})
        Bias=0
       )
      )
      Destination Space=::Simulation
      Apply Scale=true
     )
    )
   ) {isSetParameters=true, module=SetVariables_#2}
  ]
  rendererClasses [&NiagaraRibbonRendererProperties]
  renderers [{rendererClass=&NiagaraRibbonRendererProperties}]
 }
 Retrotrail1002 {
  emitterUpdateScript [
   &EmitterState()
   &SpawnPerUnit(Module.Use Max Movement Threshold=true, Module.Use Movement Tolerance=false, Use Spawn Probability=false) {enabled=false}
   &VC_FilteredSocketCount()
   &VC_RateQualityHandler()
   &VC_SpawnRibbon_v1(Number of Ribbons=$Emitter.PortCount, SpawnRate=60)
  ]
  particleSpawnScript [
   &VC_PaintedVariantsOverride(
    Black=<rgba 0.0156 0.0156 0.0156 1>
    BurntSienna=<rgba 0.1562 0.0504 0.022 0.8>
    Cobalt=<rgba 0.0522 0.138 0.9115 0.8348>
    Crimson=<rgba 0.4271 0.0356 0.0462 0.5217>
    DefaultColor=<rgba 0.1284 0.3065 1 1>
    ForestGreen=<rgba 0 0.5 0.0253 0.4>
    Grey=<rgba 0.2083 0.2083 0.2083 1>
    Lime=<rgba 0.4075 0.6094 0.0508 0.8522>
    Orange=<rgba 0.6724 0.2346 0 0.8>
    Pink=<rgba 0.7 0.1896 0.5144 0.8>
    Purple=<rgba 0.1913 0 0.6795 1>
    R1 Black:LinearColor=<rgba 0.2708 0.2708 0.2708 1>
    R1 BurntSienna:LinearColor=<rgba 0.8229 0.2696 0.0129 1>
    R1 Cobalt:LinearColor=<rgba 0 0.185 10 1>
    R1 Crimson:LinearColor=<rgba 1 0 0.0271 1>
    R1 DefaultColor=<rgba 4 0.28 0.28 1>
    R1 Grey:LinearColor=<rgba 0.7478 0.7478 0.7478 1>
    R1 Lime:LinearColor=<rgba 1.5 3 0 1>
    R1 Orange:LinearColor=<rgba 5 1.3346 0 1>
    R1 Pink:LinearColor=<rgba 3 0.5 2.0909 1>
    R1 Purple:LinearColor=<rgba 0.4 0 2 1>
    R1 Saffron:LinearColor=<rgba 3 2.3 0 1>
    R1 SkyBlue:LinearColor=<rgba 0 1.9962 3 1>
    R1 TitaniumWhite:LinearColor=<rgba 1 1 1 1>
    R2 Black:LinearColor=<rgba 0.0365 0.0365 0.0365 1>
    R2 BurntSienna:LinearColor=<rgba 0.2083 0.0709 0.0335 1>
    R2 DefaultColor:LinearColor=<rgba 0.1284 0.3065 1 1>
    Saffron=<rgba 0.8698 0.6774 0.0453 0.6957>
    SkyBlue=<rgba 0 0.6654 1 1>
    TitaniumWhite=<rgba 0.8854 0.8854 0.8854 1>
    Use R1 Painted Overrides=true
    Use R2 Painted Overrides=true
   )
   &VC_InitializeRibbon_v1(
    Color=$Particles.Color
    Custom Offset=true
    Enable Skeletal Position=true
    Lifetime=0.8
    Local Vector Offset=<-100 0 5>
    Offset Amount=1
    Ribbon Width=60
   )
   &AddVelocity(Velocity=<-1200 0 0>, Velocity Speed Scale=1.4) {module=AddVelocity001}
  ]
  particleUpdateScript [
   &ParticleState()
   &InheritSourceMovement(
    Applied Position Scale=&Multiply_VectorByFloat(
     Vector=<1 1 1>
     Float=&FloatFromCurve(FloatCurve=DataInterface() {propertyValues {Curve {
         defaultValue=1
         keys [{interpMode=RCIM_Cubic, tangentMode=RCTM_User, time=0, value=1}, {interpMode=RCIM_Cubic, tangentMode=RCTM_User, time=1, value=0}]
        }}})
    )
    ApplyForce=false
    ApplyPosition=true
    ApplyPositionFalloff=false
   )
   &SolveForcesAndVelocity(Clamp Velocity=false, Limit Acceleration=false)
   &VC_UpdateRibbon_v1(Enable Skeletal Position=true)
   /SetVariables(
    Particles.PosNoodle1:NiagaraPosition=&AddVectorToPosition(
     Position=$Particles.Position
     Vector=&TransformVector(
      Vector=&MakeVector(
       X=&Sine(
        Normalized Angle=&Add_Float(
         A=&Multiply_Float(A=$Particles.NormalizedAge, B=-1.3)
         B=&Multiply_Float(A=$Emitter.LoopedAge, B=0.8)
        )
        Period=1
        Scale=40
        Bias=0
       )
       Y=&Sine(
        Normalized Angle=&Add_Float(
         A=&Multiply_Float(A=$Particles.NormalizedAge, B=-1.3)
         B=&Multiply_Float(A=$Emitter.LoopedAge, B=1)
        )
        Period=1.5
        Scale=5
        Bias=0
       )
       Z=&Sine(
        Normalized Angle=&Add_Float(
         A=&Multiply_Float(A=$Particles.NormalizedAge, B=-1.3)
         B=&Multiply_Float(A=$Emitter.LoopedAge, B=0.8)
        )
        Period=1.5
        Scale=5
        Bias=0
       )
      )
      Destination Space=::Simulation
      Apply Scale=true
     )
    )
    Particles.PosNoodle2:NiagaraPosition=&AddVectorToPosition(
     Position=$Particles.Position
     Vector=&TransformVector(
      Vector=&MakeVector(
       X=0
       Y=&Sine(
        Normalized Angle=&Add_Float(
         A=&Multiply_Float(A=$Particles.NormalizedAge, B=-1.3)
         B=&Multiply_Float(A=$Emitter.LoopedAge, B=1.8)
        )
        Period=0.4
        Scale=&FloatFromCurve(FloatCurve=DataInterface() {propertyValues {Curve {keys [
             {arriveTangent=41.4286, interpMode=RCIM_Cubic, time=0, value=0.3}
             {arriveTangentWeight=3.3375, interpMode=RCIM_Cubic, leaveTangentWeight=3.3375, tangentMode=RCTM_User, time=0.3, value=20}
             {arriveTangent=28.1747, interpMode=RCIM_Cubic, leaveTangent=28.1747, tangentMode=RCTM_User, time=1, value=50}
            ]}}})
        Bias=0
       )
       Z=&Sine(
        Normalized Angle=&Add_Float(
         A=&Multiply_Float(A=$Particles.NormalizedAge, B=-1.3)
         B=&Multiply_Float(A=$Emitter.LoopedAge, B=2)
        )
        Period=0.8
        Scale=&FloatFromCurve(FloatCurve=DataInterface() {propertyValues {Curve {keys [
             <at 0.3 0.3 RCIM_Cubic>
             {
              arriveTangent=10.3659
              arriveTangentWeight=3.3375
              interpMode=RCIM_Cubic
              leaveTangent=10.3659
              leaveTangentWeight=3.3375
              tangentMode=RCTM_User
              time=0.5
              value=20
             }
             {arriveTangent=75.1908, interpMode=RCIM_Cubic, leaveTangent=75.1908, tangentMode=RCTM_User, time=1, value=60}
            ]}}})
        Bias=0
       )
      )
      Destination Space=::Simulation
      Apply Scale=true
     )
    )
    Particles.PosNoodle3:NiagaraPosition=&AddVectorToPosition(
     Position=$Particles.Position
     Vector=&TransformVector(
      Vector=&MakeVector(
       X=0
       Y=&Sine(
        Normalized Angle=&Add_Float(
         A=&Multiply_Float(A=$Particles.NormalizedAge, B=-1.3)
         B=&Multiply_Float(A=$Emitter.LoopedAge, B=2.1)
        )
        Period=0.5
        Scale=&FloatFromCurve(FloatCurve=DataInterface() {propertyValues {Curve {keys [
             {arriveTangent=39.5768, interpMode=RCIM_Cubic, leaveTangent=39.5768, tangentMode=RCTM_User, time=0, value=0.3}
             {
              arriveTangent=32.917
              arriveTangentWeight=3.3375
              interpMode=RCIM_Cubic
              leaveTangent=32.917
              leaveTangentWeight=3.3375
              time=0.595
              value=8.803
             }
             {interpMode=RCIM_Cubic, leaveTangent=75.1908, time=1, value=33.217}
            ]}}})
        Bias=0
       )
       Z=&Sine(
        Normalized Angle=&Add_Float(
         A=&Multiply_Float(A=$Particles.NormalizedAge, B=-1.3)
         B=&Multiply_Float(A=$Emitter.LoopedAge, B=1.7)
        )
        Period=0.5
        Scale=&FloatFromCurve(FloatCurve=DataInterface() {propertyValues {Curve {keys [
             <at 0 0 RCIM_Cubic>
             {arriveTangentWeight=3.3375, interpMode=RCIM_Cubic, leaveTangentWeight=3.3375, tangentMode=RCTM_User, time=0.3, value=1}
             {arriveTangent=75.1908, interpMode=RCIM_Cubic, leaveTangent=75.1908, tangentMode=RCTM_User, time=1, value=45}
            ]}}})
        Bias=0
       )
      )
      Destination Space=::Simulation
      Apply Scale=true
     )
    )
    Particles.PosNoodle4:NiagaraPosition=&AddVectorToPosition(
     Position=$Particles.Position
     Vector=&TransformVector(
      Vector=&MakeVector(
       X=0
       Y=&Sine(
        Normalized Angle=&Add_Float(
         A=&Multiply_Float(A=$Particles.NormalizedAge, B=-1.3)
         B=&Multiply_Float(A=$Emitter.LoopedAge, B=1.5)
        )
        Period=0.4
        Scale=&FloatFromCurve(FloatCurve=DataInterface() {propertyValues {Curve {keys [
             {arriveTangent=41.4286, interpMode=RCIM_Cubic, time=0, value=0.3}
             {arriveTangentWeight=3.3375, interpMode=RCIM_Cubic, leaveTangentWeight=3.3375, tangentMode=RCTM_User, time=0.2, value=11.4906}
             {arriveTangent=31.4052, interpMode=RCIM_Cubic, leaveTangent=31.4052, tangentMode=RCTM_User, time=1, value=26}
            ]}}})
        Bias=0
       )
       Z=&Sine(
        Normalized Angle=&Add_Float(
         A=&Multiply_Float(A=$Particles.NormalizedAge, B=-1.3)
         B=&Multiply_Float(A=$Emitter.LoopedAge, B=1)
        )
        Period=1
        Scale=&FloatFromCurve(FloatCurve=DataInterface() {propertyValues {Curve {keys [
             <at 0 1 RCIM_Cubic>
             {arriveTangentWeight=6.6125, interpMode=RCIM_Cubic, leaveTangentWeight=6.725, tangentMode=RCTM_User, time=0.6, value=20}
             {interpMode=RCIM_Cubic, leaveTangent=75.1908, time=1, value=35}
            ]}}})
        Bias=0
       )
      )
      Destination Space=::Simulation
      Apply Scale=true
     )
    )
   ) {isSetParameters=true, module=SetVariables_#2}
  ]
  rendererClasses [&NiagaraRibbonRendererProperties]
  renderers [{rendererClass=&NiagaraRibbonRendererProperties}]
 }
 Retrotrail1006 {
  emitterUpdateScript [
   &EmitterState()
   &SpawnPerUnit(Module.Use Max Movement Threshold=true, Module.Use Movement Tolerance=false, Use Spawn Probability=false) {enabled=false}
   &VC_FilteredSocketCount()
   &VC_RateQualityHandler()
   &VC_SpawnRibbon_v1(Number of Ribbons=$Emitter.PortCount, SpawnRate=60)
  ]
  particleSpawnScript [
   &VC_PaintedVariantsOverride(
    Black=<rgba 0.0156 0.0156 0.0156 1>
    BurntSienna=<rgba 0.1562 0.0504 0.022 0.8>
    Cobalt=<rgba 0.0522 0.138 0.9115 0.8348>
    Crimson=<rgba 0.4271 0.0356 0.0462 0.5217>
    DefaultColor=<rgba 0.1284 0.3065 1 1>
    ForestGreen=<rgba 0 0.5 0.0253 0.4>
    Grey=<rgba 0.2083 0.2083 0.2083 1>
    Lime=<rgba 0.4075 0.6094 0.0508 0.8522>
    Orange=<rgba 0.6724 0.2346 0 0.8>
    Pink=<rgba 0.7 0.1896 0.5144 0.8>
    Purple=<rgba 0.1913 0 0.6795 1>
    R1 Black:LinearColor=<rgba 0.2708 0.2708 0.2708 1>
    R1 BurntSienna:LinearColor=<rgba 0.8229 0.2696 0.0129 1>
    R1 Cobalt:LinearColor=<rgba 0 0.185 10 1>
    R1 Crimson:LinearColor=<rgba 1 0 0.0271 1>
    R1 DefaultColor=<rgba 4 0.28 0.28 1>
    R1 Grey:LinearColor=<rgba 0.7478 0.7478 0.7478 1>
    R1 Lime:LinearColor=<rgba 1.5 3 0 1>
    R1 Orange:LinearColor=<rgba 5 1.3346 0 1>
    R1 Pink:LinearColor=<rgba 3 0.5 2.0909 1>
    R1 Purple:LinearColor=<rgba 0.4 0 2 1>
    R1 Saffron:LinearColor=<rgba 3 2.3 0 1>
    R1 SkyBlue:LinearColor=<rgba 0 1.9962 3 1>
    R1 TitaniumWhite:LinearColor=<rgba 1 1 1 1>
    R2 Black:LinearColor=<rgba 0.0365 0.0365 0.0365 1>
    R2 BurntSienna:LinearColor=<rgba 0.2083 0.0709 0.0335 1>
    R2 DefaultColor:LinearColor=<rgba 0.1284 0.3065 1 1>
    Saffron=<rgba 0.8698 0.6774 0.0453 0.6957>
    SkyBlue=<rgba 0 0.6654 1 1>
    TitaniumWhite=<rgba 0.8854 0.8854 0.8854 1>
    Use R1 Painted Overrides=true
    Use R2 Painted Overrides=true
   )
   &VC_InitializeRibbon_v1(
    Color=$Particles.Color
    Custom Offset=true
    Enable Skeletal Position=true
    Lifetime=0.8
    Local Vector Offset=<-100 0 4>
    Offset Amount=1
    Ribbon Width=60
   )
   &AddVelocity(Velocity=<-1200 0 0>, Velocity Speed Scale=1.4) {module=AddVelocity001}
  ]
  particleUpdateScript [
   &ParticleState()
   &InheritSourceMovement(
    Applied Position Scale=&Multiply_VectorByFloat(
     Vector=<1 1 1>
     Float=&FloatFromCurve(FloatCurve=DataInterface() {propertyValues {Curve {
         defaultValue=1
         keys [{interpMode=RCIM_Cubic, tangentMode=RCTM_User, time=0, value=1}, {interpMode=RCIM_Cubic, tangentMode=RCTM_User, time=1, value=0}]
        }}})
    )
    ApplyForce=false
    ApplyPosition=true
    ApplyPositionFalloff=false
   )
   &SolveForcesAndVelocity(Clamp Velocity=false, Limit Acceleration=false)
   &VC_UpdateRibbon_v1(Enable Skeletal Position=true)
   /SetVariables(
    Particles.PosNoodle1:NiagaraPosition=&AddVectorToPosition(
     Position=$Particles.Position
     Vector=&TransformVector(
      Vector=&MakeVector(
       X=&Sine(
        Normalized Angle=&Add_Float(
         A=&Multiply_Float(A=$Particles.NormalizedAge, B=-1.3)
         B=&Multiply_Float(A=$Emitter.LoopedAge, B=1)
        )
        Period=1
        Scale=40
        Bias=0
       )
       Y=&Sine(
        Normalized Angle=&Add_Float(
         A=&Multiply_Float(A=$Particles.NormalizedAge, B=-1.3)
         B=&Multiply_Float(A=$Emitter.LoopedAge, B=1)
        )
        Period=1.5
        Scale=5
        Bias=0
       )
       Z=&Sine(
        Normalized Angle=&Add_Float(
         A=&Multiply_Float(A=$Particles.NormalizedAge, B=-1.3)
         B=&Multiply_Float(A=$Emitter.LoopedAge, B=0.8)
        )
        Period=1.5
        Scale=5
        Bias=0
       )
      )
      Destination Space=::Simulation
      Apply Scale=true
     )
    )
    Particles.PosNoodle2:NiagaraPosition=&AddVectorToPosition(
     Position=$Particles.Position
     Vector=&TransformVector(
      Vector=&MakeVector(
       X=0
       Y=&Sine(
        Normalized Angle=&Add_Float(
         A=&Multiply_Float(A=$Particles.NormalizedAge, B=-1.3)
         B=&Multiply_Float(A=$Emitter.LoopedAge, B=1.8)
        )
        Period=0.4
        Scale=&FloatFromCurve(FloatCurve=DataInterface() {propertyValues {Curve {keys [
             {arriveTangent=41.4286, interpMode=RCIM_Cubic, time=0, value=0.3}
             {arriveTangentWeight=3.3375, interpMode=RCIM_Cubic, leaveTangentWeight=3.3375, tangentMode=RCTM_User, time=0.3, value=20}
             {arriveTangent=28.1747, interpMode=RCIM_Cubic, leaveTangent=28.1747, tangentMode=RCTM_User, time=1, value=50}
            ]}}})
        Bias=0
       )
       Z=&Sine(
        Normalized Angle=&Add_Float(
         A=&Multiply_Float(A=$Particles.NormalizedAge, B=-1.3)
         B=&Multiply_Float(A=$Emitter.LoopedAge, B=2)
        )
        Period=0.8
        Scale=&FloatFromCurve(FloatCurve=DataInterface() {propertyValues {Curve {keys [
             <at 0.3 0.3 RCIM_Cubic>
             {
              arriveTangent=10.3659
              arriveTangentWeight=3.3375
              interpMode=RCIM_Cubic
              leaveTangent=10.3659
              leaveTangentWeight=3.3375
              tangentMode=RCTM_User
              time=0.5
              value=20
             }
             {arriveTangent=75.1908, interpMode=RCIM_Cubic, leaveTangent=75.1908, tangentMode=RCTM_User, time=1, value=60}
            ]}}})
        Bias=0
       )
      )
      Destination Space=::Simulation
      Apply Scale=true
     )
    )
    Particles.PosNoodle3:NiagaraPosition=&AddVectorToPosition(
     Position=$Particles.Position
     Vector=&TransformVector(
      Vector=&MakeVector(
       X=0
       Y=&Sine(
        Normalized Angle=&Add_Float(
         A=&Multiply_Float(A=$Particles.NormalizedAge, B=-1.3)
         B=&Multiply_Float(A=$Emitter.LoopedAge, B=2.1)
        )
        Period=0.5
        Scale=&FloatFromCurve(FloatCurve=DataInterface() {propertyValues {Curve {keys [
             {arriveTangent=39.5768, interpMode=RCIM_Cubic, leaveTangent=39.5768, tangentMode=RCTM_User, time=0, value=0.3}
             {
              arriveTangent=32.917
              arriveTangentWeight=3.3375
              interpMode=RCIM_Cubic
              leaveTangent=32.917
              leaveTangentWeight=3.3375
              time=0.595
              value=8.803
             }
             {interpMode=RCIM_Cubic, leaveTangent=75.1908, time=1, value=33.217}
            ]}}})
        Bias=0
       )
       Z=&Sine(
        Normalized Angle=&Add_Float(
         A=&Multiply_Float(A=$Particles.NormalizedAge, B=-1.3)
         B=&Multiply_Float(A=$Emitter.LoopedAge, B=1.7)
        )
        Period=0.5
        Scale=&FloatFromCurve(FloatCurve=DataInterface() {propertyValues {Curve {keys [
             <at 0 0 RCIM_Cubic>
             {arriveTangentWeight=3.3375, interpMode=RCIM_Cubic, leaveTangentWeight=3.3375, tangentMode=RCTM_User, time=0.3, value=1}
             {arriveTangent=75.1908, interpMode=RCIM_Cubic, leaveTangent=75.1908, tangentMode=RCTM_User, time=1, value=45}
            ]}}})
        Bias=0
       )
      )
      Destination Space=::Simulation
      Apply Scale=true
     )
    )
    Particles.PosNoodle4:NiagaraPosition=&AddVectorToPosition(
     Position=$Particles.Position
     Vector=&TransformVector(
      Vector=&MakeVector(
       X=0
       Y=&Sine(
        Normalized Angle=&Add_Float(
         A=&Multiply_Float(A=$Particles.NormalizedAge, B=-1.3)
         B=&Multiply_Float(A=$Emitter.LoopedAge, B=1.5)
        )
        Period=0.4
        Scale=&FloatFromCurve(FloatCurve=DataInterface() {propertyValues {Curve {keys [
             {arriveTangent=41.4286, interpMode=RCIM_Cubic, time=0, value=0.3}
             {arriveTangentWeight=3.3375, interpMode=RCIM_Cubic, leaveTangentWeight=3.3375, tangentMode=RCTM_User, time=0.2, value=11.4906}
             {arriveTangent=31.4052, interpMode=RCIM_Cubic, leaveTangent=31.4052, tangentMode=RCTM_User, time=1, value=26}
            ]}}})
        Bias=0
       )
       Z=&Sine(
        Normalized Angle=&Add_Float(
         A=&Multiply_Float(A=$Particles.NormalizedAge, B=-1.3)
         B=&Multiply_Float(A=$Emitter.LoopedAge, B=1)
        )
        Period=1
        Scale=&FloatFromCurve(FloatCurve=DataInterface() {propertyValues {Curve {keys [
             <at 0 1 RCIM_Cubic>
             {arriveTangentWeight=6.6125, interpMode=RCIM_Cubic, leaveTangentWeight=6.725, tangentMode=RCTM_User, time=0.6, value=20}
             {interpMode=RCIM_Cubic, leaveTangent=75.1908, time=1, value=35}
            ]}}})
        Bias=0
       )
      )
      Destination Space=::Simulation
      Apply Scale=true
     )
    )
   ) {isSetParameters=true, module=SetVariables_#2}
  ]
  rendererClasses [&NiagaraRibbonRendererProperties]
  renderers [{rendererClass=&NiagaraRibbonRendererProperties}]
 }
}
summary {
 systemName=NS_Boost_RetroStrips
 emitters [
  {emitterName=Pyramid, rendererClasses [&NiagaraMeshRendererProperties, &NiagaraSpriteRendererProperties]}
  {emitterName=Retrotrail1001, rendererClasses [&NiagaraRibbonRendererProperties]}
  {emitterName=Retrotrail1002, rendererClasses [&NiagaraRibbonRendererProperties]}
  {emitterName=Retrotrail1006, rendererClasses [&NiagaraRibbonRendererProperties]}
 ]
 userVariables [
  User.bIsFrontendPreview {type=NiagaraBool}
  User.bIsPainted {type=NiagaraBool}
  User.ColorVariantID {type=NiagaraInt32}
  User.PaintedColor {type=LinearColor, description="", defaultValue {struct=LinearColor, value=<rgba 1 1 1 1>}}
  User.PaintedScalar {type=NiagaraFloat, defaultValue=1, description=""}
  User.PortCount {type=NiagaraInt32, defaultValue=1}
 ]
}
